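# Control Rig for the ShrimpStroll medium male player body, driving corrective deformation and forearm metal bars.
# The forward solve runs the shared deform-rig function, then a Metal_bars_IK function covering two left and four right bar chains.
# Parent constraints pin each chain's end control to its hand, except right midtop, which follows middle_02_r with maintained offset.
# Stretch-enabled three-bone IK without pole vectors solves each chain toward its control.
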ControlRigBlueprint /BRCosmetics/Characters/Player/Male/Medium/Bodies/M_MED_ShrimpStroll/Meshes/M_MED_ShrimpStroll_CtrlRig
graphs [
 New Function_ContainedGraph {nodes [
   n1 = RigVMFunctionEntryNode()
   RigVMFunction_Sequence = &RigVMFunction_Sequence(ExecuteContext=@n1.ExecuteContext)
   &RigUnit_ParentConstraint(
    AdvancedSettings=(InterpolationType=Average,RotationOrderForFilter=XZY)
    Child=(Type=Control,Name="metal_bars_04_outer_l_ctrl")
    Filter=(TranslationFilter=(bX=True,bY=True,bZ=True),RotationFilter=(bX=true,bY=true,bZ=true),ScaleFilter=(bX=True,bY=True,bZ=True))
    Parents=((Item=(Type=Bone,Name="hand_l"),Weight=1))
    Weight=1
    ExecutePin=@RigVMFunction_Sequence.A
   )
   > IKThreeBone = RigVMFunctionReferenceNode(
    Bone A=(Type=Bone,Name="metal_bars_01_outer_l")
    Bone B=(Type=Bone,Name="metal_bars_02_outer_l")
    Bone C=(Type=None,Name="metal_bars_03_outer_l")
    Bone D=(Type=Bone,Name="metal_bars_04_outer_l")
    Enable Stretch=true
    Foot Ctrl=(Type=Control,Name="metal_bars_04_outer_l_ctrl")
    Pole Vector Node=(Type=Control,Name="None")
    Primary Axis=(X=1,Y=0,Z=0)
    Secondary Axis=(X=0,Y=-1,Z=0)
    Stretch Maximum Ratio=1.25
    Thigh Ctrl Optional=(Type=None,Name="None")
    scale=(X=0,Y=0,Z=0)
   )
   > &RigUnit_ParentConstraint(
    AdvancedSettings=(InterpolationType=Average,RotationOrderForFilter=XZY)
    Child=(Type=Control,Name="metal_bars_04_inner_l_ctrl")
    Filter=(TranslationFilter=(bX=True,bY=True,bZ=True),RotationFilter=(bX=true,bY=true,bZ=true),ScaleFilter=(bX=True,bY=True,bZ=True))
    Parents=((Item=(Type=Bone,Name="hand_l"),Weight=1))
    Weight=1
   )
   > IKThreeBone_1 = RigVMFunctionReferenceNode(
    Bone A=(Type=Bone,Name="metal_bars_01_inner_l")
    Bone B=(Type=Bone,Name="metal_bars_02_inner_l")
    Bone C=(Type=Bone,Name="metal_bars_03_inner_l")
    Bone D=(Type=Bone,Name="metal_bars_04_inner_l")
    Enable Stretch=true
    Foot Ctrl=(Type=Control,Name="metal_bars_04_inner_l_ctrl")
    Pole Vector Node=(Type=Control,Name="None")
    Primary Axis=(X=1,Y=0,Z=0)
    Secondary Axis=(X=0,Y=-1,Z=0)
    Stretch Maximum Ratio=1.25
    Thigh Ctrl Optional=(Type=None,Name="None")
    scale=(X=0,Y=0,Z=0)
   )
   > &RigUnit_ParentConstraint(
    AdvancedSettings=(InterpolationType=Average,RotationOrderForFilter=XZY)
    Child=(Type=Control,Name="metal_bars_04_top_r_ctrl")
    Filter=(TranslationFilter=(bX=True,bY=True,bZ=True),RotationFilter=(bX=true,bY=true,bZ=true),ScaleFilter=(bX=True,bY=True,bZ=True))
    Parents=((Item=(Type=Bone,Name="hand_r"),Weight=1))
    Weight=1
   )
   > IKThreeBone_1_1 = RigVMFunctionReferenceNode(
    Bone A=(Type=Bone,Name="metal_bars_01_top_r")
    Bone B=(Type=Bone,Name="metal_bars_02_top_r")
    Bone C=(Type=Bone,Name="metal_bars_03_top_r")
    Bone D=(Type=Bone,Name="metal_bars_04_top_r")
    Enable Stretch=true
    Foot Ctrl=(Type=Control,Name="metal_bars_04_top_r_ctrl")
    Pole Vector Node=(Type=Control,Name="None")
    Primary Axis=(X=1,Y=0,Z=0)
    Secondary Axis=(X=0,Y=-1,Z=0)
    Stretch Maximum Ratio=1.25
    Thigh Ctrl Optional=(Type=None,Name="None")
    scale=(X=0,Y=0,Z=0)
   )
   > &RigUnit_ParentConstraint(
    AdvancedSettings=(InterpolationType=Average,RotationOrderForFilter=XZY)
    Child=(Type=Control,Name="metal_bars_03_midtop_r_ctrl")
    Filter=(TranslationFilter=(bX=True,bY=True,bZ=True),RotationFilter=(bX=true,bY=true,bZ=true),ScaleFilter=(bX=True,bY=True,bZ=True))
    Parents=((Item=(Type=Bone,Name="middle_02_r"),Weight=1))
    Weight=1
    bMaintainOffset=true
   )
   > IKThreeBone_1_1_1 = RigVMFunctionReferenceNode(
    Bone A=(Type=Bone,Name="metal_bars_01_midtop_r")
    Bone B=(Type=Bone,Name="metal_bars_02_midtop_r")
    Bone C=(Type=None,Name="None")
    Bone D=(Type=Bone,Name="metal_bars_03_midtop_r")
    Enable Stretch=true
    Foot Ctrl=(Type=Control,Name="metal_bars_03_midtop_r_ctrl")
    Pole Vector Node=(Type=Control,Name="None")
    Primary Axis=(X=1,Y=0,Z=0)
    Secondary Axis=(X=0,Y=-1,Z=0)
    Stretch Maximum Ratio=1.25
    Thigh Ctrl Optional=(Type=None,Name="None")
    scale=(X=0,Y=0,Z=0)
   )
   > &RigUnit_ParentConstraint(
    AdvancedSettings=(InterpolationType=Average,RotationOrderForFilter=XZY)
    Child=(Type=Control,Name="metal_bars_04_inner_r_ctrl")
    Filter=(TranslationFilter=(bX=True,bY=True,bZ=True),RotationFilter=(bX=true,bY=true,bZ=true),ScaleFilter=(bX=True,bY=True,bZ=True))
    Parents=((Item=(Type=Bone,Name="hand_r"),Weight=1))
    Weight=1
   )
   > IKThreeBone_1_1_2 = RigVMFunctionReferenceNode(
    Bone A=(Type=Bone,Name="metal_bars_01_inner_r")
    Bone B=(Type=Bone,Name="metal_bars_02_inner_r")
    Bone C=(Type=Bone,Name="metal_bars_03_inner_r")
    Bone D=(Type=Bone,Name="metal_bars_04_inner_r")
    Enable Stretch=true
    Foot Ctrl=(Type=Control,Name="metal_bars_04_inner_r_ctrl")
    Pole Vector Node=(Type=Control,Name="None")
    Primary Axis=(X=1,Y=0,Z=0)
    Secondary Axis=(X=0,Y=-1,Z=0)
    Stretch Maximum Ratio=1.25
    Thigh Ctrl Optional=(Type=None,Name="None")
    scale=(X=0,Y=0,Z=0)
   )
   > &RigUnit_ParentConstraint(
    AdvancedSettings=(InterpolationType=Average,RotationOrderForFilter=XZY)
    Child=(Type=Control,Name="metal_bars_04_outer_r_ctrl")
    Filter=(TranslationFilter=(bX=True,bY=True,bZ=True),RotationFilter=(bX=true,bY=true,bZ=true),ScaleFilter=(bX=True,bY=True,bZ=True))
    Parents=((Item=(Type=Bone,Name="hand_r"),Weight=1))
    Weight=1
   )
   > IKThreeBone_1_1_2_1 = RigVMFunctionReferenceNode(
    Bone A=(Type=Bone,Name="metal_bars_01_outer_r")
    Bone B=(Type=Bone,Name="metal_bars_02_outer_r")
    Bone C=(Type=Bone,Name="metal_bars_03_outer_r")
    Bone D=(Type=Bone,Name="metal_bars_04_outer_r")
    Enable Stretch=true
    Foot Ctrl=(Type=Control,Name="metal_bars_04_outer_r_ctrl")
    Pole Vector Node=(Type=Control,Name="None")
    Primary Axis=(X=1,Y=0,Z=0)
    Secondary Axis=(X=0,Y=-1,Z=0)
    Stretch Maximum Ratio=1.25
    Thigh Ctrl Optional=(Type=None,Name="None")
    scale=(X=0,Y=0,Z=0)
   )
   RigVMFunctionReturnNode(ExecuteContext=@RigVMFunction_Sequence.B)
  ]}
 RigVMFunctionLibrary {nodes [Metal_bars_IK = RigVMCollapseNode(ExecuteContext=())]}
 RigVMModel {nodes [
   RigUnit_BeginExecution = &RigUnit_BeginExecution()
   FN_DeformRig_FNC = RigVMFunctionReferenceNode(
    DeformRig_ Curve="(deformRig_l_upperarm_up_alpha_72_#1=true,deformRig_r_upperarm_up_alpha_73_#2=true,deformRig_l_upperarm_fr_alpha_74_#3=true,deformRig_r_upperarm_fr_alpha_75_CB51B80C4 ...(truncated)"
    DeformRig_Bool="(upperarm_1_#4=true,elbow_3_#5=true,wrist_7_#6=True,groin_8_#7=true,glute_17_#8=true,knee_9_40C408324CA261 ...(truncated)"
    DeformRig_Debug=(upperarm_debug_18_#4=False,wrist_debug_20_#6=False,lat_debug_26_#9=False,groinGlute_debug_27_#7=False)
    DeformRig_Value="(upperarm_up_18_#10=(Rotation=(X=0,Y=0,Z=0,W=1),Translation=(X=0,Y=0,Z=0.8),Scale3D=(X=0.7,Y=1,Z=1)),upperarm_fwd_fr_19_#11=(Rotation=(X=0,Y=0,Z=0,W=1),Translation=(X=0,Y=10,Z=0),Scale3D=(X=0.5,Y=1, ...(truncated)"
    ExecuteContext=@RigUnit_BeginExecution.ExecutePin
   )
   > Metal_bars_IK = RigVMFunctionReferenceNode()
  ]}
]
hierarchy [
 root {loc=<0 0 0>, rot=<0 0 0 1>}
 pelvis {parent=root, loc=<0 -0.7071 93.1191>, rot=<0.0127 0.707 -0.0127 -0.707>}
 spine_01 {parent=pelvis, loc=<6.7494 -0.6388 0>, rot=<0 0 0.034 -0.9994>}
 spine_02 {parent=spine_01, loc=<7.2518 -0.035 0>, rot=<0 0 0.0238 -0.9997>}
 spine_03 {parent=spine_02, loc=<6.0104 0 0>, rot=<0 0 0.0149 -0.9999>}
 spine_04 {parent=spine_03, loc=<8.1024 0 0>, rot=<0 0 -0.0132 -0.9999>}
 spine_05 {parent=spine_04, loc=<10.0146 0 0>, rot=<0 0 -0.0015 -1>}
 attach_cape {parent=spine_05, loc=<9.017 -11.714 0>, rot=<0.0616 -0.7044 -0.0616 -0.7044>}
 clavicle_l {parent=spine_05, loc=<4.3575 -0.1466 -2.3223>, rot=<0.0465 -0.7289 -0.009 -0.6829>}
 upperarm_l {parent=clavicle_l, loc=<15.5111 0 0>, rot=<0.0226 -0.3234 -0.0088 -0.9459>}
 lowerarm_l {parent=upperarm_l, loc=<26.6514 0 0>, rot=<0.0391 -0.0099 -0.2453 -0.9686>}
 hand_l {parent=lowerarm_l, loc=<22.3491 0 0>, rot=<-0.7081 -0.0507 0.0975 -0.6975>}
 thumb_01_l {parent=hand_l, loc=<1.3118 -0.3191 -2.459>, rot=<0.6546 -0.4042 0.0664 -0.6354>}
 thumb_02_l {parent=thumb_01_l, loc=<5.9336 -0.0448 0.0401>, rot=<-0.0037 0.0213 0.298 -0.9543>}
 thumb_03_l {parent=thumb_02_l, loc=<2.2223 0.3682 -0.0907>, rot=<0.0118 -0.013 0.0066 -0.9998>}
 index_metacarpal_l {parent=hand_l, loc=<1.6535 0.0183 -2.96>, rot=<0.0158 -0.0124 -0.0173 -0.9996>}
 index_01_l {parent=index_metacarpal_l, loc=<8.3008 0.0357 -0.0524>, rot=<-0.0044 0.0485 0.0924 -0.9945>}
 index_02_l {parent=index_01_l, loc=<3.3587 -0.3004 0.0452>, rot=<-0.002 0.0027 0.3225 -0.9466>}
 index_03_l {parent=index_02_l, loc=<2.5578 -0.123 -0.0252>, rot=<-0.002 0.0067 0.2558 -0.9667>}
 middle_metacarpal_l {parent=hand_l, loc=<2.1856 -0.1725 -0.5984>, rot=<0.006 0.0168 -0.0329 -0.9993>}
 middle_01_l {parent=middle_metacarpal_l, loc=<7.3275 -0.2447 0.0129>, rot=<-0.0208 0.0042 0.0858 -0.9961>}
 middle_02_l {parent=middle_01_l, loc=<4.0123 0.1401 0.1175>, rot=<-0.0218 0.0258 0.3512 -0.9357>}
 middle_03_l {parent=middle_02_l, loc=<2.6077 0.0534 0.1373>, rot=<0.0082 0.0188 0.1708 -0.9851>}
 ring_metacarpal_l {parent=hand_l, loc=<2.2074 -0.3625 1.3638>, rot=<-0.0284 0.0722 -0.0458 -0.9959>}
 ring_01_l {parent=ring_metacarpal_l, loc=<7.0295 -0.6197 -0.1446>, rot=<-0.0489 -0.0252 0.1892 -0.9804>}
 ring_02_l {parent=ring_01_l, loc=<3.6174 0.4647 -0.1027>, rot=<0 0 0.2806 -0.9598>}
 ring_03_l {parent=ring_02_l, loc=<2.4152 0.114 0.0199>, rot=<0 0 0.1189 -0.9929>}
 pinky_metacarpal_l {parent=hand_l, loc=<1.7616 -0.9753 3.3814>, rot=<-0.0397 0.0814 -0.069 -0.9935>}
 pinky_01_l {parent=pinky_metacarpal_l, loc=<7.1709 -0.4831 -0.1094>, rot=<-0.0654 -0.0035 0.2242 -0.9723>}
 pinky_02_l {parent=pinky_01_l, loc=<2.8274 0.0893 -0.1043>, rot=<0.0037 0.0011 0.2854 -0.9584>}
 pinky_03_l {parent=pinky_02_l, loc=<2.2046 0.0307 -0.0078>, rot=<-0.0002 0.0098 0.1445 -0.9895>}
 lowerarm_twist_02_l {parent=lowerarm_l, loc=<9.6199 0 0>, rot=<0 0 0 -1>}
 lowerarm_twist_01_l {parent=lowerarm_l, loc=<17.8549 0 0>, rot=<0 0 0 -1>}
 base_leather_flaps_Lowerarm_01_l {parent=lowerarm_twist_01_l, loc=<-9.1175 -0.4213 -3.7217>, rot=<0.71 -0.6693 -0.1106 -0.1891>}
 dyn_leather_flaps_Lowerarm_01_l {parent=base_leather_flaps_Lowerarm_01_l, loc=<0 0 0>, rot=<0 0 0 1>}
 base_leather_flaps_Lowerarm_02_l {parent=lowerarm_twist_01_l, loc=<-2.3409 -4.2724 1.2398>, rot=<-0.3295 -0.3921 0.5582 -0.6528>}
 dyn_leather_flaps_Lowerarm_02_l {parent=base_leather_flaps_Lowerarm_02_l, loc=<0 0 0>, rot=<0 0 0 1>}
 deform_elbow_in_base_l {parent=lowerarm_l, loc=<0.2112 5.259 0.8274>, rot=<0 0 0 -1>}
 deform_elbow_in_l {parent=deform_elbow_in_base_l, loc=<0 0 0>, rot=<0 0 0 1>}
 deform_elbow_out_base_l {parent=lowerarm_l, loc=<0 -4 0>, rot=<0 0 0 -1>}
 deform_elbow_out_l {parent=deform_elbow_out_base_l, loc=<0 0 0>, rot=<0 0 0 1>}
 metal_bars_01_outer_l {parent=lowerarm_l, loc=<8.6522 -4.6123 -0.8>, rot=<0.0479 0.2167 0.1009 -0.9698>}
 metal_bars_02_outer_l {parent=metal_bars_01_outer_l, loc=<9.2553 0.0298 0.1824>, rot=<0.0456 -0.3508 -0.0353 -0.9347>}
 metal_bars_03_outer_l {parent=metal_bars_02_outer_l, loc=<7.1701 -0.0192 -0.0031>, rot=<0.0256 -0.0621 -0.0465 -0.9967>}
 metal_bars_04_outer_l {parent=metal_bars_03_outer_l, loc=<6.2375 0.3393 -0.0122>, rot=<-0.0007 -0.0986 -0.263 -0.9597>}
 metal_bars_05_outer_l {parent=metal_bars_04_outer_l, loc=<0 0 0>, rot=<0 0 0 1>}
 metal_bars_01_inner_l {parent=lowerarm_l, loc=<8.7141 4.7896 2.8193>, rot=<0.416 0.0517 -0.0935 -0.9031>}
 metal_bars_02_inner_l {parent=metal_bars_01_inner_l, loc=<9.1866 -0.0097 0.3405>, rot=<-0.2271 -0.1499 0.1388 -0.9522>}
 metal_bars_03_inner_l {parent=metal_bars_02_inner_l, loc=<8.9951 0.0648 0.38>, rot=<-0.0223 -0.2474 0.0587 -0.9669>}
 metal_bars_04_inner_l {parent=metal_bars_03_inner_l, loc=<0 0 0>, rot=<0 0 0 1>}
 upperarm_twist_01_l {parent=upperarm_l, loc=<0.0041 0 0>, rot=<0 0 0 -1>}
 upperarm_twist_02_l {parent=upperarm_l, loc=<13.7941 0 0>, rot=<0 0 0 -1>}
 base_leather_flaps_Upperarm_01_l {parent=upperarm_twist_02_l, loc=<-2.6127 -1.8583 3.1147>, rot=<0.0408 0.0189 0.6889 -0.7234>}
 dyn_leather_flaps_Upperarm_01_l {parent=base_leather_flaps_Upperarm_01_l, loc=<0 0 0>, rot=<0 0 0 1>}
 base_leather_flaps_Upperarm_02_l {parent=upperarm_twist_02_l, loc=<0.6456 -1.4034 3.3071>, rot=<0.0413 0.0538 0.6702 -0.7391>}
 dyn_leather_flaps_Upperarm_02_l {parent=base_leather_flaps_Upperarm_02_l, loc=<0 0 0>, rot=<0 0 0 1>}
 attach_backpack {parent=spine_05, loc=<-0.123 -13.473 0>, rot=<-0.0093 -0.707 0.0093 -0.707>}
 neck_01 {parent=spine_05, loc=<11.3138 0.1529 0>, rot=<0 0 -0.2003 -0.9797>}
 neck_02 {parent=neck_01, loc=<4.5627 0 0>, rot=<0 0 0.0504 -0.9987>}
 head {parent=neck_02, loc=<5.3914 0 0>, rot=<0 0 0.115 -0.9934>}
 clavicle_r {parent=spine_05, loc=<4.3575 -0.1466 2.3223>, rot=<0.7289 0.0465 -0.6829 0.009>}
 upperarm_r {parent=clavicle_r, loc=<-15.5111 0 0>, rot=<0.0226 -0.3234 -0.0088 -0.9459>}
 lowerarm_r {parent=upperarm_r, loc=<-26.6514 0 0>, rot=<0.0391 -0.0099 -0.2453 -0.9686>}
 hand_r {parent=lowerarm_r, loc=<-22.3491 0 0>, rot=<-0.7081 -0.0507 0.0975 -0.6975>}
 thumb_01_r {parent=hand_r, loc=<-1.3118 0.3191 2.459>, rot=<0.6546 -0.4042 0.0664 -0.6354>}
 thumb_02_r {parent=thumb_01_r, loc=<-5.9336 0.0448 -0.0401>, rot=<-0.0037 0.0213 0.298 -0.9543>}
 thumb_03_r {parent=thumb_02_r, loc=<-2.2223 -0.3682 0.0907>, rot=<0.0118 -0.013 0.0066 -0.9998>}
 index_metacarpal_r {parent=hand_r, loc=<-1.6535 -0.0183 2.96>, rot=<0.0158 -0.0124 -0.0173 -0.9996>}
 index_01_r {parent=index_metacarpal_r, loc=<-8.3008 -0.0357 0.0524>, rot=<-0.0044 0.0485 0.0924 -0.9945>}
 index_02_r {parent=index_01_r, loc=<-3.3587 0.3004 -0.0452>, rot=<-0.002 0.0027 0.3225 -0.9466>}
 index_03_r {parent=index_02_r, loc=<-2.5578 0.123 0.0252>, rot=<-0.002 0.0067 0.2558 -0.9667>}
 middle_metacarpal_r {parent=hand_r, loc=<-2.1856 0.1725 0.5984>, rot=<0.006 0.0168 -0.0329 -0.9993>}
 middle_01_r {parent=middle_metacarpal_r, loc=<-7.3275 0.2447 -0.0129>, rot=<-0.0208 0.0042 0.0858 -0.9961>}
 middle_02_r {parent=middle_01_r, loc=<-4.0123 -0.1401 -0.1175>, rot=<-0.0218 0.0258 0.3512 -0.9357>}
 middle_03_r {parent=middle_02_r, loc=<-2.6077 -0.0534 -0.1373>, rot=<0.0082 0.0188 0.1708 -0.9851>}
 ring_metacarpal_r {parent=hand_r, loc=<-2.2074 0.3625 -1.3638>, rot=<-0.0284 0.0722 -0.0458 -0.9959>}
 ring_01_r {parent=ring_metacarpal_r, loc=<-7.0295 0.6197 0.1446>, rot=<-0.0489 -0.0252 0.1892 -0.9804>}
 ring_02_r {parent=ring_01_r, loc=<-3.6174 -0.4647 0.1027>, rot=<0 0 0.2806 -0.9598>}
 ring_03_r {parent=ring_02_r, loc=<-2.4152 -0.114 -0.0199>, rot=<0 0 0.1189 -0.9929>}
 pinky_metacarpal_r {parent=hand_r, loc=<-1.7616 0.9753 -3.3814>, rot=<-0.0397 0.0814 -0.069 -0.9935>}
 pinky_01_r {parent=pinky_metacarpal_r, loc=<-7.1709 0.4831 0.1094>, rot=<-0.0654 -0.0035 0.2242 -0.9723>}
 pinky_02_r {parent=pinky_01_r, loc=<-2.8274 -0.0893 0.1043>, rot=<0.0037 0.0011 0.2854 -0.9584>}
 pinky_03_r {parent=pinky_02_r, loc=<-2.2046 -0.0307 0.0078>, rot=<-0.0002 0.0098 0.1445 -0.9895>}
 lowerarm_twist_02_r {parent=lowerarm_r, loc=<-9.6199 0 0>, rot=<0 0 0 -1>}
 base_leather_flaps_Lowerarm_01_r {parent=lowerarm_twist_02_r, loc=<-0.0635 -4.9382 -1.5598>, rot=<-0.5172 -0.4697 0.5455 -0.463>}
 dyn_leather_flaps_Lowerarm_01_r {parent=base_leather_flaps_Lowerarm_01_r, loc=<0 0 0>, rot=<0 0 0 1>}
 lowerarm_twist_01_r {parent=lowerarm_r, loc=<-17.8549 0 0>, rot=<0 0 0 -1>}
 deform_elbow_in_base_r {parent=lowerarm_r, loc=<-0.2111 -5.2591 -0.8275>, rot=<0 0 0 -1>}
 deform_elbow_in_r {parent=deform_elbow_in_base_r, loc=<-0.0001 0 0.0002>, rot=<0 0 0 1>}
 deform_elbow_out_base_r {parent=lowerarm_r, loc=<0 4 -0.0001>, rot=<0 0 0 -1>}
 deform_elbow_out_r {parent=deform_elbow_out_base_r, loc=<0 0 0>, rot=<0 0 0 1>}
 metal_bars_01_inner_r {parent=lowerarm_r, loc=<-9.5837 2.0146 3.4036>, rot=<0.4045 -0.8908 0.2067 0.0032>}
 metal_bars_02_inner_r {parent=metal_bars_01_inner_r, loc=<9.031 -0.4979 0.0296>, rot=<0.0832 -0.1329 -0.4945 -0.8549>}
 metal_bars_03_inner_r {parent=metal_bars_02_inner_r, loc=<9.7881 0.2211 0.396>, rot=<0 0 0 1>}
 metal_bars_04_inner_r {parent=metal_bars_03_inner_r, loc=<0 0 0>, rot=<0 0 0 1>}
 metal_bars_01_outer_r {parent=lowerarm_r, loc=<-11.4157 3.7289 1.1152>, rot=<0.3863 0.4468 0.8006 -0.1009>}
 metal_bars_02_outer_r {parent=metal_bars_01_outer_r, loc=<8.491 -0.2999 -0.3875>, rot=<-0.1573 -0.1902 0.6027 -0.7589>}
 metal_bars_03_outer_r {parent=metal_bars_02_outer_r, loc=<8.1386 -0.6283 -1.611>, rot=<0 -0.0951 0 -0.9955>}
 metal_bars_04_outer_r {parent=metal_bars_03_outer_r, loc=<0 0 0>, rot=<0 0 0 1>}
 metal_bars_01_top_r {parent=lowerarm_r, loc=<-11.769 -3.8179 -2.9586>, rot=<-0.1432 -0.9207 -0.0177 -0.3627>}
 metal_bars_02_top_r {parent=metal_bars_01_top_r, loc=<8.3482 -0.1833 -0.4744>, rot=<0.1462 -0.6951 0.1601 -0.6854>}
 metal_bars_03_top_r {parent=metal_bars_02_top_r, loc=<12.0764 0.7709 -0.521>, rot=<-0.0263 -0.0011 -0.1078 -0.9938>}
 metal_bars_04_top_r {parent=metal_bars_03_top_r, loc=<0 0 0>, rot=<0 0 0 1>}
 metal_bars_01_midtop_r {parent=lowerarm_r, loc=<-12.1791 -0.4074 -7.1143>, rot=<-0.0069 0.6836 0.7251 0.0828>}
 metal_bars_02_midtop_r {parent=metal_bars_01_midtop_r, loc=<23.5791 2.2754 0.0146>, rot=<0 0 0 1>}
 metal_bars_03_midtop_r {parent=metal_bars_02_midtop_r, loc=<0 0 0>, rot=<0 0 0 1>}
 upperarm_twist_01_r {parent=upperarm_r, loc=<-0.0041 0 0>, rot=<0 0 0 -1>}
 upperarm_twist_02_r {parent=upperarm_r, loc=<-13.7941 0 0>, rot=<0 0 0 -1>}
 thigh_l {parent=pelvis, loc=<-2.5693 -0.0109 -7.7501>, rot=<0.0012 0.0572 0.0155 -0.9982>}
 calf_l {parent=thigh_l, loc=<-39.0484 0 0>, rot=<0.0006 -0.0161 -0.0391 -0.9991>}
 foot_l {parent=calf_l, loc=<-42.561 0.1077 0.4292>, rot=<0.0418 -0.0455 0.0343 -0.9975>}
 ball_l {parent=foot_l, loc=<-6.1599 13.8264 0.6806>, rot=<0.0001 -0.0001 -0.7063 -0.7079>}
 calf_twist_02_l {parent=calf_l, loc=<-18.8064 -0.0967 0.0397>, rot=<0.0237 0.0095 -0.0035 -0.9997>}
 calf_twist_01_l {parent=calf_l, loc=<-35.813 -0.0967 0.0397>, rot=<0.0237 0.0095 -0.0035 -0.9997>}
 deform_knee_in_base_l {parent=calf_l, loc=<1.298 -5.2957 -0.0496>, rot=<0 0 0 -1>}
 deform_knee_in_l {parent=deform_knee_in_base_l, loc=<0 0 0>, rot=<0 0 0 1>}
 deform_knee_out_base_l {parent=calf_l, loc=<0 4 0>, rot=<0 0 0 -1>}
 deform_knee_out_l {parent=deform_knee_out_base_l, loc=<0 0 0>, rot=<0 0 0 1>}
 thigh_twist_01_l {parent=thigh_l, loc=<-21.6303 0 0>, rot=<0.0013 0 0 -1>}
 base_leather_flaps_thigh_01_l {parent=thigh_twist_01_l, loc=<6.3873 -1.6696 -7.8106>, rot=<0.6416 -0.7195 -0.192 -0.1837>}
 dyn_leather_flaps_thigh_01_l {parent=base_leather_flaps_thigh_01_l, loc=<0 0 0>, rot=<0 0 0 1>}
 base_leather_flaps_thigh_02_l {parent=thigh_twist_01_l, loc=<0.6534 -3.2576 -7.4014>, rot=<0.6416 -0.7195 -0.192 -0.1837>}
 dyn_leather_flaps_thigh_02_l {parent=base_leather_flaps_thigh_02_l, loc=<0 0 0>, rot=<0 0 0 1>}
 deform_groin_base_l {parent=thigh_l, loc=<-2.1908 8.1142 -1.7442>, rot=<0 0 0 -1>}
 deform_groin_l {parent=deform_groin_base_l, loc=<0 0 0>, rot=<0 0 0 1>}
 thigh_r {parent=pelvis, loc=<-2.5693 -0.0109 7.7501>, rot=<0.0572 -0.0012 0.9982 0.0155>}
 calf_r {parent=thigh_r, loc=<39.0484 0 0>, rot=<0.0006 -0.0161 -0.0391 -0.9991>}
 foot_r {parent=calf_r, loc=<42.561 -0.1077 -0.4292>, rot=<0.0418 -0.0455 0.0343 -0.9975>}
 ball_r {parent=foot_r, loc=<6.1599 -13.8264 -0.6806>, rot=<0.0001 -0.0001 -0.7063 -0.7079>}
 calf_twist_02_r {parent=calf_r, loc=<18.8064 0.0967 -0.0397>, rot=<0.0237 0.0095 -0.0035 -0.9997>}
 base_leather_flaps_calf_01_l {parent=calf_twist_02_r, loc=<8.7896 -3.0172 2.3435>, rot=<-0.4648 0.2305 -0.5106 -0.6857>}
 dyn_leather_flaps_calf_01_l {parent=base_leather_flaps_calf_01_l, loc=<0 0 0>, rot=<0 0 0 1>}
 calf_twist_01_r {parent=calf_r, loc=<35.813 0.0967 -0.0397>, rot=<0.0237 0.0095 -0.0035 -0.9997>}
 base_leather_flaps_foot_01_r {parent=calf_twist_01_r, loc=<3.0187 -2.2891 -4.003>, rot=<0.9426 -0.1043 -0.1537 0.2775>}
 dyn_leather_flaps_foot_01_r {parent=base_leather_flaps_foot_01_r, loc=<0 0 0>, rot=<0 0 0 1>}
 dyn_leather_flaps_foot_02_r {parent=dyn_leather_flaps_foot_01_r, loc=<3.9656 0.0524 0.0167>, rot=<0 0 0.0467 -0.9989>}
 dyn_leather_flaps_foot_03_r {parent=dyn_leather_flaps_foot_02_r, loc=<4.2336 0.1006 0.0395>, rot=<-0.0252 0.0021 -0.0149 -0.9996>}
 knee_off_r {parent=calf_r, loc=<13.5675 4.5983 3.6078>, rot=<0.7204 0.1347 0.6782 0.0549>}
 knee_offA_r {parent=knee_off_r, loc=<0 0 0>, rot=<0 0 0 1>}
 deform_knee_in_base_r {parent=calf_r, loc=<-1.298 5.2957 0.0496>, rot=<0 0 0 -1>}
 deform_knee_in_r {parent=deform_knee_in_base_r, loc=<0 0 0>, rot=<0 0 0 1>}
 deform_knee_out_base_r {parent=calf_r, loc=<0 -4 0>, rot=<0 0 0 -1>}
 deform_knee_out_r {parent=deform_knee_out_base_r, loc=<0 0 0>, rot=<0 0 0 1>}
 thigh_twist_01_r {parent=thigh_r, loc=<21.6303 0 0>, rot=<0.0013 0 0 -1>}
 deform_groin_base_r {parent=thigh_r, loc=<2.1908 -8.1142 1.7442>, rot=<0 0 0 -1>}
 deform_groin_r {parent=deform_groin_base_r, loc=<0 0 0>, rot=<0 0 0 1>}
 deform_glute_base_l {parent=pelvis, loc=<-8.8064 -8.8222 -8.4287>, rot=<0.0275 0.0501 -0.4472 -0.8926>}
 deform_glute_l {parent=deform_glute_base_l, loc=<0 0 0>, rot=<0 0 0 1>}
 deform_glute_base_r {parent=pelvis, loc=<-8.8062 -8.8222 8.4287>, rot=<0.0501 -0.0275 0.8926 -0.4472>}
 deform_glute_r {parent=deform_glute_base_r, loc=<0 0 0>, rot=<0 0 0 1>}
 base_leather_flaps_a_Back_r {parent=pelvis, loc=<-7.8583 -12.466 7.5692>, rot=<-0.0191 -0.6948 0.7188 0.0122>}
 dyn_leather_flaps_Back_a_01_r {parent=base_leather_flaps_a_Back_r, loc=<0 0 0>, rot=<0 0 0 1>}
 dyn_leather_flaps_Back_a_02_r {parent=dyn_leather_flaps_Back_a_01_r, loc=<4.0177 0.0438 0.2355>, rot=<0 0 0.0296 -0.9996>}
 dyn_leather_flaps_Back_a_03_r {parent=dyn_leather_flaps_Back_a_02_r, loc=<5.9695 -0.0354 -0.0643>, rot=<0 0 -0.02 -0.9998>}
 dyn_leather_flaps_Back_a_04_r {parent=dyn_leather_flaps_Back_a_03_r, loc=<4.6226 -0.1756 0.0877>, rot=<0 0 0 1>}
 dyn_leather_flaps_Back_a_05_r {parent=dyn_leather_flaps_Back_a_04_r, loc=<5.3913 -0.1915 0.0725>, rot=<0 0 0 1>}
 Pos_a_Back_r {parent=pelvis, loc=<-7.8583 -12.4661 7.5692>, rot=<-0.0127 -0.707 0.0127 -0.707>}
 base_leather_flaps_thig_r {parent=pelvis, loc=<-15.3195 4.199 11.0984>, rot=<-0.299 -0.1326 -0.7189 -0.6134>}
 dyn_leather_flaps_thig_r {parent=base_leather_flaps_thig_r, loc=<0 0 0>, rot=<0 0 0 1>}
 base_leather_flap_02_hip_l {parent=pelvis, loc=<6.3883 4.6362 -9.7313>, rot=<-0.0369 -0.9226 -0.3806 -0.0514>}
 dyn_leather_flap_02_hip_l {parent=base_leather_flap_02_hip_l, loc=<0 0 0>, rot=<0 0 0 1>}
 dyn_leather_flap_03_hip_l {parent=dyn_leather_flap_02_hip_l, loc=<1.6639 -1.1217 0.0098>, rot=<0 0 0 1>}
 base_leather_flap_01_hip_l {parent=pelvis, loc=<6.1509 -1.9853 -9.7313>, rot=<-0.0179 -0.9996 0.0221 -0.0054>}
 dyn_leather_flap_01_hip_l {parent=base_leather_flap_01_hip_l, loc=<0 0 0>, rot=<0 0 0 1>}
 base_fleather_flaps_hip_b_01_l {parent=pelvis, loc=<8.1541 -3.6096 -12.9272>, rot=<0.0253 -0.9557 0.1737 -0.2361>}
 dyn_fleather_flaps_hip_b_01_l {parent=base_fleather_flaps_hip_b_01_l, loc=<0 0 0>, rot=<0 0 0 1>}
 dyn_fleather_flaps_hip_b_02_l {parent=dyn_fleather_flaps_hip_b_01_l, loc=<5.1787 0 0>, rot=<0 0 0 1>}
 base_fleather_flaps_hip_a_01_l {parent=pelvis, loc=<7.9935 5.062 -11.9906>, rot=<-0.0611 -0.9361 -0.3405 -0.064>}
 dyn_fleather_flaps_hip_a_01_l {parent=base_fleather_flaps_hip_a_01_l, loc=<0 0 0>, rot=<0 0 0 1>}
 dyn_fleather_flaps_hip_a_02_l {parent=dyn_fleather_flaps_hip_a_01_l, loc=<5.2529 0 0>, rot=<0 0 0 1>}
 dyn_fleather_flaps_hip_a_03_l {parent=dyn_fleather_flaps_hip_a_02_l, loc=<3.309 0.5676 -0.4886>, rot=<0 -0.077 0 -0.997>}
 base_fleather_flaps_hip_a_01_m {parent=pelvis, loc=<6.6478 10.0378 -0.4087>, rot=<-0.0152 -0.7223 -0.6913 0.011>}
 dyn_fleather_flaps_hip_a_01_m {parent=base_fleather_flaps_hip_a_01_m, loc=<0 0 0>, rot=<0 0 0 1>}
 dyn_fleather_flaps_hip_a_02_m {parent=dyn_fleather_flaps_hip_a_01_m, loc=<4.7539 0 0>, rot=<0 0 0 1>}
 base_fleather_flaps_hip_a_01_r {parent=pelvis, loc=<8.6213 6.1831 9.6339>, rot=<0.2051 0.1871 0.959 0.0573>}
 dyn_fleather_flaps_hip_a_01_r {parent=base_fleather_flaps_hip_a_01_r, loc=<0 0 0>, rot=<0 0 0 1>}
 dyn_fleather_flaps_hip_a_02_r {parent=dyn_fleather_flaps_hip_a_01_r, loc=<5.5709 0 0>, rot=<0 0 0 1>}
 base_fleather_flaps_hip_b_01_r {parent=pelvis, loc=<8.0944 -3.1281 12.6004>, rot=<0.2812 -0.2241 0.9318 -0.0504>}
 dyn_fleather_flaps_hip_b_01_r {parent=base_fleather_flaps_hip_b_01_r, loc=<0 0 0>, rot=<0 0 0 1>}
 dyn_fleather_flaps_hip_b_02_r {parent=dyn_fleather_flaps_hip_b_01_r, loc=<5.5438 0 0>, rot=<0 0 0 1>}
 hand_l_ctrl {controlType=EULER_TRANSFORM, type=Control, loc=<0 0 0>, rot=<0 0 0 1>}
 metal_bars_04_outer_l_ctrl {controlType=EULER_TRANSFORM, type=Control, loc=<0 0 0>, rot=<0 0 0 1>}
 metal_bars_04_inner_l_ctrl {controlType=EULER_TRANSFORM, type=Control, loc=<0 0 0>, rot=<0 0 0 1>}
 metal_bars_04_top_r_ctrl {controlType=EULER_TRANSFORM, type=Control, loc=<0 0 0>, rot=<0 0 0 1>}
 hand_r_ctrl {controlType=EULER_TRANSFORM, type=Control, loc=<0 0 0>, rot=<0 0 0 1>}
 metal_bars_03_midtop_r_ctrl {controlType=EULER_TRANSFORM, type=Control, loc=<0 0 0>, rot=<0 0 0 1>}
 metal_bars_04_inner_r_ctrl {controlType=EULER_TRANSFORM, type=Control, loc=<0 0 0>, rot=<0 0 0 1>}
 metal_bars_04_outer_r_ctrl {controlType=EULER_TRANSFORM, type=Control, loc=<0 0 0>, rot=<0 0 0 1>}
 middle_02_r_ctrl {controlType=EULER_TRANSFORM, type=Control, loc=<0 0 0>, rot=<0 0 0 1>}
]
variables [Iscrouching_deform {type=double, default=0}]
