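# Physics asset for the M_MED_ShrimpStroll male medium player body cosmetic, set up for right-side leather flaps on the back and foot.
# Capsules cover the pelvis, upper thighs, calves and feet.
# Two dyn_ capsule chains hang from base bones, five segments off the pelvis and three off the right calf.
# Bases have free angular constraints, while chain links allow limited swing with zero twist; calves constrain directly to the pelvis.
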PhysicsAsset /BRCosmetics/Characters/Player/Male/Medium/Bodies/M_MED_ShrimpStroll/Meshes/M_MED_ShrimpStroll_Physics
bodies {
 base_leather_flaps_a_Back_r [{shapeType=Capsule, length=0.505, radius=0.505, shapeName=base_leather_flaps_a_Back_r_capsule, center=<2 0 0>}]
 base_leather_flaps_foot_01_r [{
   shapeType=Capsule
   length=3
   radius=1.5
   shapeName=dyn_coat_arm_strap_mstr_l_capsule
   center=<2 0 1.0302>
   rotation=<rot 90 179.9736 179.9737>
  }]
 calf_l [{shapeType=Capsule, length=29.1242, radius=7.9157, center=<-20.5385 -2.0341 0.0219>, rotation=<rot 90 0 0>}]
 calf_r [{shapeType=Capsule, length=29.1242, radius=7.9157, center=<20.5385 2.0341 -0.0219>, rotation=<rot 90 0 180>}]
 dyn_leather_flaps_Back_a_01_r [{shapeType=Capsule, length=3, radius=1.5, shapeName=dyn_coat_arm_strap_mstr_l_capsule, center=<2 0 0.5>, rotation=<rot 90 0 0>}]
 dyn_leather_flaps_Back_a_02_r [{shapeType=Capsule, length=3, radius=1.5, shapeName=dyn_coat_arm_strap_02_l_capsule, center=<2 0 0.5>, rotation=<rot 90 0 0>}]
 dyn_leather_flaps_Back_a_03_r [{shapeType=Capsule, length=3, radius=1.5, shapeName=dyn_coat_arm_strap_03_l_capsule, center=<2 0 0.5>, rotation=<rot 90 0 0>}]
 dyn_leather_flaps_Back_a_04_r [{shapeType=Capsule, length=3, radius=1.5, shapeName=dyn_coat_arm_strap_03_l_capsule, center=<2 0 0.5>, rotation=<rot 90 0 0>}]
 dyn_leather_flaps_Back_a_05_r [{shapeType=Capsule, length=3, radius=1.5, shapeName=dyn_coat_arm_strap_03_l_capsule, center=<2 0 0.5>, rotation=<rot 90 0 0>}]
 dyn_leather_flaps_foot_01_r [{
   shapeType=Capsule
   length=3
   radius=1.5
   shapeName=dyn_coat_arm_strap_mstr_l_capsule
   center=<2 0 1.0302>
   rotation=<rot 90 179.9736 179.9737>
  }]
 dyn_leather_flaps_foot_02_r [{shapeType=Capsule, length=3, radius=1.5, shapeName=dyn_coat_arm_strap_02_l_capsule, center=<2 0 1.0302>, rotation=<rot 90 179.9951 179.9951>}]
 dyn_leather_flaps_foot_03_r [{
   shapeType=Capsule
   length=3
   radius=1.5
   shapeName=dyn_coat_arm_strap_03_l_capsule
   center=<2 0 1.0302>
   rotation=<rot 90 -178.3124 -178.3124>
  }]
 foot_l [{shapeType=Capsule, length=19.1865, radius=5.7269, center=<-4.9208 7.1124 -0.9818>, rotation=<rot 0 1.5681 -90>}]
 foot_r [{shapeType=Capsule, length=19.1865, radius=5.7269, center=<4.9208 -7.1124 0.9818>, rotation=<rot 0 1.5681 90>}]
 pelvis [{shapeType=Capsule, length=10.9585, radius=13.4601, shapeName=pelvis_capsule, center=<-1.9198 -0.6554 0.2372>}]
 thigh_twist_01_l [{shapeType=Capsule, length=20.1046, radius=9.9538, center=<-0.9373 -0.9752 -0.723>, rotation=<rot 90 0 0>}]
 thigh_twist_01_r [{shapeType=Capsule, length=20.1046, radius=9.9538, center=<0.9373 0.9752 0.723>, rotation=<rot 90 0 180>}]
}
constraints [
 {bone1Name=thigh_twist_01_l, bone2Name=pelvis, swing1LimitDegrees=45, swing2LimitDegrees=45, twistLimitDegrees=45}
 {bone1Name=thigh_twist_01_r, bone2Name=pelvis, swing1LimitDegrees=45, swing2LimitDegrees=45, twistLimitDegrees=45}
 {bone1Name=calf_r, bone2Name=pelvis, swing1LimitDegrees=45, swing2LimitDegrees=45, twistLimitDegrees=45}
 {bone1Name=foot_r, bone2Name=calf_r, swing1LimitDegrees=45, swing2LimitDegrees=45, twistLimitDegrees=45}
 {
  bone1Name=base_leather_flaps_a_Back_r
  bone2Name=pelvis
  swing1LimitDegrees=45
  swing1Motion=Free
  swing2LimitDegrees=45
  swing2Motion=Free
  twistLimitDegrees=45
  twistMotion=Free
 }
 {
  bone1Name=dyn_leather_flaps_Back_a_01_r
  bone2Name=base_leather_flaps_a_Back_r
  swing1LimitDegrees=35
  swing2LimitDegrees=20
  twistLimitDegrees=0
 }
 {
  bone1Name=dyn_leather_flaps_Back_a_02_r
  bone2Name=dyn_leather_flaps_Back_a_01_r
  swing1LimitDegrees=35
  swing2LimitDegrees=35
  twistLimitDegrees=0
 }
 {
  bone1Name=dyn_leather_flaps_Back_a_03_r
  bone2Name=dyn_leather_flaps_Back_a_02_r
  swing1LimitDegrees=35
  swing2LimitDegrees=35
  twistLimitDegrees=0
 }
 {
  bone1Name=dyn_leather_flaps_Back_a_04_r
  bone2Name=dyn_leather_flaps_Back_a_03_r
  swing1LimitDegrees=35
  swing2LimitDegrees=35
  twistLimitDegrees=0
 }
 {
  bone1Name=dyn_leather_flaps_Back_a_05_r
  bone2Name=dyn_leather_flaps_Back_a_04_r
  swing1LimitDegrees=35
  swing2LimitDegrees=35
  twistLimitDegrees=0
 }
 {
  bone1Name=base_leather_flaps_foot_01_r
  bone2Name=calf_r
  swing1LimitDegrees=45
  swing1Motion=Free
  swing2LimitDegrees=45
  swing2Motion=Free
  twistLimitDegrees=45
  twistMotion=Free
 }
 {
  bone1Name=dyn_leather_flaps_foot_01_r
  bone2Name=base_leather_flaps_foot_01_r
  swing1LimitDegrees=35
  swing2LimitDegrees=35
  twistLimitDegrees=0
 }
 {
  bone1Name=dyn_leather_flaps_foot_02_r
  bone2Name=dyn_leather_flaps_foot_01_r
  swing1LimitDegrees=35
  swing2LimitDegrees=35
  twistLimitDegrees=0
 }
 {
  bone1Name=dyn_leather_flaps_foot_03_r
  bone2Name=dyn_leather_flaps_foot_02_r
  swing1LimitDegrees=35
  swing2LimitDegrees=35
  twistLimitDegrees=0
 }
 {bone1Name=calf_l, bone2Name=pelvis, swing1LimitDegrees=45, swing2LimitDegrees=45, twistLimitDegrees=45}
 {
  bone1Name=foot_l
  bone2Name=calf_l
  swing1LimitDegrees=45
  swing1Motion=Free
  swing2LimitDegrees=45
  swing2Motion=Free
  twistLimitDegrees=45
  twistMotion=Free
 }
]
