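# Animation Blueprint for the male ShrimpStroll player outfit, adding secondary motion and corrective deformation to its body mesh.
# The parent's copied pose feeds a Control Rig, RigidBody, weighted multi-bone constraints, per-bone AnimDynamics springs for leather flaps, then knee and metal-bar offsets.
# Updates set gravity, knee-fix, simulation-space and alpha variables from movement, DBNO-carried and front-end states.
# On customization load, it compares the loadout pickaxe against two specific items to set gauntlet values.
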AnimBlueprint /BRCosmetics/Characters/Player/Male/Medium/Bodies/M_MED_ShrimpStroll/Meshes/M_MED_ShrimpStroll_AnimBP
target_skeleton=&/M_MED_ShrimpStroll_Skeleton
graphs [
 AnimGraph {nodes [
   n1 = Control Rig!(Source=Copy Pose From Mesh() {bUseAttachedParent=true}, Alpha, Iscrouching_deform:double=Get IsCrouching() -> (IsCrouching:double)) -> (Pose) {
    alphaBoolBlend {blendOption=Linear}
    controlRigAssetReference {blueprintRigClass=&/M_MED_ShrimpStroll_CtrlRig.M_MED_ShrimpStroll_CtrlRig_C}
   }
   n2 = RigidBody!(ComponentPose=Local To Component(LocalPose=@n1), Alpha, OverrideWorldGravity:Vector="0,0,0", ExternalForce) -> (Pose) {
    bOverrideWorldGravity=true
    overridePhysicsAsset=&/M_MED_ShrimpStroll_Physics
    simulationSpace=BaseBoneSpace
    alphaBoolBlend {blendOption=Linear}
    baseBoneRef {boneName=pelvis}
    componentAppliedLinearAccClamp=<3000 3000 3000>
    componentLinearAccScale=<0.3 0.3 0.3>
    componentLinearVelScale=<0.3 0.3 0.3>
    simSpaceSettings {
     maxAngularAcceleration=1000
     maxAngularVelocity=3
     maxLinearAcceleration=700
     maxLinearVelocity=500
     velocityScaleZ=0
     worldAlpha=0.35
     externalLinearDragV=<0.2 0.2>
    }
   }
   n3 = "Constraint - base_leather_flaps_a_Back_r (PPPPP)"!(
    ComponentPose=@n2
    Alpha=1
    ConstraintWeights_0=0.687
    ConstraintWeights_1:float=0.2619
    ConstraintWeights_2:float=0.0448
    ConstraintWeights_3:float=0.0059
    ConstraintWeights_4:float=0.0003
   ) -> (Pose) {
    alphaBoolBlend {blendOption=Linear}
    boneToModify {boneName=base_leather_flaps_a_Back_r}
    constraintSetup [
     {targetBone {boneName=pelvis}}
     {targetBone {boneName=thigh_twist_01_r}}
     {targetBone {boneName=deform_glute_r}}
     {targetBone {boneName=thigh_twist_01_l}}
     {targetBone {boneName=spine_01}}
    ]
    constraintWeights [1,1,1,1,1]
   }
   n4 = "Constraint - base_leather_flaps_thig_r (PPP)"!(ComponentPose=@n3, Alpha=1, ConstraintWeights_0=0.9599, ConstraintWeights_1:float=0.0235, ConstraintWeights_2:float=0.0166) -> (Pose) {
    alphaBoolBlend {blendOption=Linear}
    boneToModify {boneName=base_leather_flaps_thig_r}
    constraintSetup [{targetBone {boneName=thigh_twist_01_r}}, {targetBone {boneName=thigh_r}}, {targetBone {boneName=pelvis}}]
    constraintWeights [1,1,1]
   }
   n5 = "Constraint - dyn_leather_flap_01_hip_l (PPPP)"!(
    ComponentPose=@n4
    Alpha=1
    ConstraintWeights_0=0.5713
    ConstraintWeights_1:float=0.3647
    ConstraintWeights_2:float=0.047
    ConstraintWeights_3:float=0.0171
   ) -> (Pose) {
    alphaBoolBlend {blendOption=Linear}
    boneToModify {boneName=dyn_leather_flap_01_hip_l}
    constraintSetup [{targetBone {boneName=pelvis}}, {targetBone {boneName=spine_01}}, {targetBone {boneName=spine_02}}, {targetBone {boneName=spine_03}}]
    constraintWeights [1,1,1,1]
   }
   n6 = "Constraint - dyn_fleather_flaps_hip_b_01_l (PPPPPP)"!(
    ComponentPose=@n5
    Alpha=1
    ConstraintWeights_0=0.6446
    ConstraintWeights_1:float=0.2959
    ConstraintWeights_2:float=0.0305
    ConstraintWeights_3:float=0.0102
    ConstraintWeights_4:float=0.0101
    ConstraintWeights_5:float=0.0088
   ) -> (Pose) {
    alphaBoolBlend {blendOption=Linear}
    boneToModify {boneName=dyn_fleather_flaps_hip_b_01_l}
    constraintSetup [
     {targetBone {boneName=pelvis}}
     {targetBone {boneName=spine_01}}
     {targetBone {boneName=spine_02}}
     {targetBone {boneName=deform_groin_l}}
     {targetBone {boneName=thigh_twist_01_l}}
     {targetBone {boneName=spine_03}}
    ]
    constraintWeights [1,1,1,1,1,1]
   }
   n7 = "Constraint - dyn_fleather_flaps_hip_b_02_l (PPPPPPP)"!(
    ComponentPose=@n6
    Alpha=1
    ConstraintWeights_0=0.7685
    ConstraintWeights_1:float=0.1104
    ConstraintWeights_2:float=0.0871
    ConstraintWeights_3:float=0.0277
    ConstraintWeights_4:float=0.0044
    ConstraintWeights_5:float=0.0018
    ConstraintWeights_6:float=0.0001
   ) -> (Pose) {
    alphaBoolBlend {blendOption=Linear}
    boneToModify {boneName=dyn_fleather_flaps_hip_b_02_l}
    constraintSetup [
     {targetBone {boneName=pelvis}}
     {targetBone {boneName=spine_01}}
     {targetBone {boneName=thigh_twist_01_l}}
     {targetBone {boneName=deform_groin_l}}
     {targetBone {boneName=spine_02}}
     {targetBone {boneName=spine_03}}
     {targetBone {boneName=deform_elbow_in_l}}
    ]
    constraintWeights [1,1,1,1,1,1,1]
   }
   n8 = "Constraint - dyn_fleather_flaps_hip_a_01_l (PPPPPPPP)"!(
    ComponentPose=@n7
    Alpha=1
    ConstraintWeights_0=0.4419
    ConstraintWeights_1:float=0.2771
    ConstraintWeights_2:float=0.1308
    ConstraintWeights_3:float=0.0704
    ConstraintWeights_4:float=0.0413
    ConstraintWeights_5:float=0.0349
    ConstraintWeights_6:float=0.0035
    ConstraintWeights_7:float=0.0002
   ) -> (Pose) {
    alphaBoolBlend {blendOption=Linear}
    boneToModify {boneName=dyn_fleather_flaps_hip_a_01_l}
    constraintSetup [
     {targetBone {boneName=pelvis}}
     {targetBone {boneName=spine_01}}
     {targetBone {boneName=deform_groin_l}}
     {targetBone {boneName=spine_02}}
     {targetBone {boneName=thigh_twist_01_l}}
     {targetBone {boneName=spine_03}}
     {targetBone {boneName=spine_04}}
     {targetBone {boneName=thigh_twist_01_r}}
    ]
    constraintWeights [1,1,1,1,1,1,1,1]
   }
   n9 = "Constraint - dyn_fleather_flaps_hip_a_02_l (PPPPPPPP)"!(
    ComponentPose=@n8
    Alpha=1
    ConstraintWeights_0=0.4051
    ConstraintWeights_1:float=0.3494
    ConstraintWeights_2:float=0.1276
    ConstraintWeights_3:float=0.1039
    ConstraintWeights_4:float=0.0091
    ConstraintWeights_5:float=0.0044
    ConstraintWeights_6:float=0.0004
    ConstraintWeights_7:float=0.0002
   ) -> (Pose) {
    alphaBoolBlend {blendOption=Linear}
    boneToModify {boneName=dyn_fleather_flaps_hip_a_02_l}
    constraintSetup [
     {targetBone {boneName=pelvis}}
     {targetBone {boneName=deform_groin_l}}
     {targetBone {boneName=thigh_twist_01_l}}
     {targetBone {boneName=spine_01}}
     {targetBone {boneName=spine_02}}
     {targetBone {boneName=spine_03}}
     {targetBone {boneName=spine_04}}
     {targetBone {boneName=thigh_twist_01_r}}
    ]
    constraintWeights [1,1,1,1,1,1,1,1]
   }
   n10 = "Constraint - dyn_fleather_flaps_hip_a_03_l (PPPPPPPP)"!(
    ComponentPose=@n9
    Alpha=1
    ConstraintWeights_0=0.0588
    ConstraintWeights_1:float=0.6542
    ConstraintWeights_2:float=0.1894
    ConstraintWeights_3:float=0.0849
    ConstraintWeights_4:float=0.0084
    ConstraintWeights_5:float=0.0036
    ConstraintWeights_6:float=0.0006
    ConstraintWeights_7:float=0.0001
   ) -> (Pose) {
    alphaBoolBlend {blendOption=Linear}
    boneToModify {boneName=dyn_fleather_flaps_hip_a_03_l}
    constraintSetup [
     {targetBone {boneName=pelvis}}
     {targetBone {boneName=deform_groin_l}}
     {targetBone {boneName=thigh_twist_01_l}}
     {targetBone {boneName=spine_01}}
     {targetBone {boneName=spine_02}}
     {targetBone {boneName=spine_03}}
     {targetBone {boneName=thigh_twist_01_r}}
     {targetBone {boneName=spine_04}}
    ]
    constraintWeights [1,1,1,1,1,1,1,1]
   }
   n11 = "Constraint - dyn_fleather_flaps_hip_a_01_m (PPPPPPPP)"!(
    ComponentPose=@n10
    Alpha=1
    ConstraintWeights_0=0.3752
    ConstraintWeights_1:float=0.3284
    ConstraintWeights_2:float=0.1623
    ConstraintWeights_3:float=0.094
    ConstraintWeights_4:float=0.015
    ConstraintWeights_5:float=0.0126
    ConstraintWeights_6:float=0.0069
    ConstraintWeights_7:float=0.0056
   ) -> (Pose) {
    alphaBoolBlend {blendOption=Linear}
    boneToModify {boneName=dyn_fleather_flaps_hip_a_01_m}
    constraintSetup [
     {targetBone {boneName=pelvis}}
     {targetBone {boneName=spine_01}}
     {targetBone {boneName=spine_02}}
     {targetBone {boneName=spine_03}}
     {targetBone {boneName=spine_04}}
     {targetBone {boneName=deform_groin_l}}
     {targetBone {boneName=deform_groin_r}}
     {targetBone {boneName=thigh_twist_01_l}}
    ]
    constraintWeights [1,1,1,1,1,1,1,1]
   }
   n12 = "Constraint - dyn_fleather_flaps_hip_a_02_m (PPPPPPPP)"!(
    ComponentPose=@n11
    Alpha=1
    ConstraintWeights_0=0.4873
    ConstraintWeights_1:float=0.2267
    ConstraintWeights_2:float=0.0804
    ConstraintWeights_3:float=0.069
    ConstraintWeights_4:float=0.039
    ConstraintWeights_5:float=0.0383
    ConstraintWeights_6:float=0.0373
    ConstraintWeights_7:float=0.0218
   ) -> (Pose) {
    alphaBoolBlend {blendOption=Linear}
    boneToModify {boneName=dyn_fleather_flaps_hip_a_02_m}
    constraintSetup [
     {targetBone {boneName=pelvis}}
     {targetBone {boneName=spine_01}}
     {targetBone {boneName=deform_groin_l}}
     {targetBone {boneName=spine_02}}
     {targetBone {boneName=thigh_twist_01_l}}
     {targetBone {boneName=deform_groin_r}}
     {targetBone {boneName=spine_03}}
     {targetBone {boneName=thigh_twist_01_r}}
    ]
    constraintWeights [1,1,1,1,1,1,1,1]
   }
   n13 = "Constraint - dyn_fleather_flaps_hip_a_01_r (PPPPPPPP)"!(
    ComponentPose=@n12
    Alpha=1
    ConstraintWeights_0=0.4364
    ConstraintWeights_1:float=0.3532
    ConstraintWeights_2:float=0.0835
    ConstraintWeights_3:float=0.0793
    ConstraintWeights_4:float=0.0344
    ConstraintWeights_5:float=0.0112
    ConstraintWeights_6:float=0.0012
    ConstraintWeights_7:float=0.0006
   ) -> (Pose) {
    alphaBoolBlend {blendOption=Linear}
    boneToModify {boneName=dyn_fleather_flaps_hip_a_01_r}
    constraintSetup [
     {targetBone {boneName=pelvis}}
     {targetBone {boneName=spine_01}}
     {targetBone {boneName=deform_groin_r}}
     {targetBone {boneName=spine_02}}
     {targetBone {boneName=spine_03}}
     {targetBone {boneName=thigh_twist_01_r}}
     {targetBone {boneName=spine_04}}
     {targetBone {boneName=deform_groin_l}}
    ]
    constraintWeights [1,1,1,1,1,1,1,1]
   }
   n14 = "Constraint - dyn_fleather_flaps_hip_a_02_r (PPPPPPPP)"!(
    ComponentPose=@n13
    Alpha=1
    ConstraintWeights_0=0.5542
    ConstraintWeights_1:float=0.1865
    ConstraintWeights_2:float=0.1318
    ConstraintWeights_3:float=0.0847
    ConstraintWeights_4:float=0.0248
    ConstraintWeights_5:float=0.0115
    ConstraintWeights_6:float=0.0042
    ConstraintWeights_7:float=0.0022
   ) -> (Pose) {
    alphaBoolBlend {blendOption=Linear}
    boneToModify {boneName=dyn_fleather_flaps_hip_a_02_r}
    constraintSetup [
     {targetBone {boneName=pelvis}}
     {targetBone {boneName=spine_01}}
     {targetBone {boneName=deform_groin_r}}
     {targetBone {boneName=thigh_twist_01_r}}
     {targetBone {boneName=spine_02}}
     {targetBone {boneName=spine_03}}
     {targetBone {boneName=deform_groin_l}}
     {targetBone {boneName=thigh_twist_01_l}}
    ]
    constraintWeights [1,1,1,1,1,1,1,1]
   }
   n15 = "Constraint - dyn_fleather_flaps_hip_b_01_r (PPPPP)"!(
    ComponentPose=@n14
    Alpha=1
    ConstraintWeights_0=0.6444
    ConstraintWeights_1:float=0.3144
    ConstraintWeights_2:float=0.0301
    ConstraintWeights_3:float=0.011
    ConstraintWeights_4:float=0.0001
   ) -> (Pose) {
    alphaBoolBlend {blendOption=Linear}
    boneToModify {boneName=dyn_fleather_flaps_hip_b_01_r}
    constraintSetup [
     {targetBone {boneName=pelvis}}
     {targetBone {boneName=spine_01}}
     {targetBone {boneName=spine_02}}
     {targetBone {boneName=spine_03}}
     {targetBone {boneName=thigh_twist_01_r}}
    ]
    constraintWeights [1,1,1,1,1]
   }
   n16 = "Constraint - dyn_fleather_flaps_hip_b_02_r (PPP)"!(ComponentPose=@n15, Alpha=1, ConstraintWeights_0=0.8349, ConstraintWeights_1:float=0.0906, ConstraintWeights_2:float=0.0745) -> (Pose) {
    alphaBoolBlend {blendOption=Linear}
    boneToModify {boneName=dyn_fleather_flaps_hip_b_02_r}
    constraintSetup [{targetBone {boneName=pelvis}}, {targetBone {boneName=spine_01}}, {targetBone {boneName=thigh_twist_01_r}}]
    constraintWeights [1,1,1]
   }
   n17 = "Constraint - metal_bars_05_outer_l (P)"!(ComponentPose=@n16, Alpha=1, ConstraintWeights_0=1, ConstraintWeights_1:float=0) -> (Pose) {
    alphaBoolBlend {blendOption=Linear}
    boneToModify {boneName=metal_bars_05_outer_l}
    constraintSetup [{targetBone {boneName=hand_l}}]
    constraintWeights [1,0]
   }
   n18 = "Constraint - metal_bars_04_inner_l (P)"!(ComponentPose=@n17, Alpha=1, ConstraintWeights_0=1, ConstraintWeights_1:float=0) -> (Pose) {
    alphaBoolBlend {blendOption=Linear}
    boneToModify {boneName=metal_bars_04_inner_l}
    constraintSetup [{targetBone {boneName=hand_l}}]
    constraintWeights [1,0]
   }
   n19 = "Constraint - metal_bars_03_midtop_r (P)"!(ComponentPose=@n18, Alpha=0, ConstraintWeights_0=1, ConstraintWeights_1:float=0) -> (Pose) {
    alphaBoolBlend {blendOption=Linear}
    boneToModify {boneName=metal_bars_03_midtop_r}
    constraintSetup [{targetBone {boneName=middle_02_r}}]
    constraintWeights [1,0]
   }
   n20 = "Constraint - metal_bars_04_inner_r (PPPPPPPP)"!(
    ComponentPose=@n19
    Alpha=1
    ConstraintWeights_0=0.5917
    ConstraintWeights_1:float=0.2801
    ConstraintWeights_2:float=0.0324
    ConstraintWeights_3:float=0.0241
    ConstraintWeights_4:float=0.0239
    ConstraintWeights_5:float=0.0239
    ConstraintWeights_6:float=0.0207
    ConstraintWeights_7:float=0.0032
   ) -> (Pose) {
    alphaBoolBlend {blendOption=Linear}
    boneToModify {boneName=metal_bars_04_inner_r}
    constraintSetup [
     {targetBone {boneName=thumb_01_r}}
     {targetBone {boneName=thumb_02_r}}
     {targetBone {boneName=hand_r}}
     {targetBone {boneName=ring_metacarpal_r}}
     {targetBone {boneName=index_metacarpal_r}}
     {targetBone {boneName=pinky_metacarpal_r}}
     {targetBone {boneName=middle_metacarpal_r}}
     {targetBone {boneName=lowerarm_twist_01_r}}
    ]
    constraintWeights [1,1,1,1,1,1,1,1]
   }
   n21 = "Constraint - metal_bars_04_top_r (PPPPPPPP)"!(
    ComponentPose=@n20
    Alpha=1
    ConstraintWeights_0=0.3928
    ConstraintWeights_1:float=0.2381
    ConstraintWeights_2:float=0.1835
    ConstraintWeights_3:float=0.1258
    ConstraintWeights_4:float=0.0424
    ConstraintWeights_5:float=0.0123
    ConstraintWeights_6:float=0.003
    ConstraintWeights_7:float=0.0022
   ) -> (Pose) {
    alphaBoolBlend {blendOption=Linear}
    boneToModify {boneName=metal_bars_04_top_r}
    constraintSetup [
     {targetBone {boneName=thumb_01_r}}
     {targetBone {boneName=hand_r}}
     {targetBone {boneName=index_metacarpal_r}}
     {targetBone {boneName=lowerarm_twist_01_r}}
     {targetBone {boneName=thumb_02_r}}
     {targetBone {boneName=middle_metacarpal_r}}
     {targetBone {boneName=ring_metacarpal_r}}
     {targetBone {boneName=lowerarm_twist_02_r}}
    ]
    constraintWeights [1,1,1,1,1,1,1,1]
   }
   n22 = Anim Dynamics - Bone: dyn_fleather_flaps_hip_b_01_l!(SimulationSpace:AnimPhysSimSpaceType Enum=Component, ComponentPose=@n21, Alpha=1, GravityOverride:Vector="0,0,-980", ExternalForce="0,0,0") -> (Pose) {
    angularDampingOverride=1
    angularSpringConstant=20000
    bAngularSpring=true
    bOverrideAngularDamping=true
    bUseGravityOverride=true
    gravityScale=0.35
    alphaBoolBlend {blendOption=Linear}
    boundBone {boneName=dyn_fleather_flaps_hip_b_01_l}
    componentAppliedLinearAccClamp=<3000 3000 3000>
    componentLinearAccScale=<0.3 0.3 0.3>
    componentLinearVelScale=<1 1 1>
    gravityOverride {z=-980}
    physicsBodyDefinitions [{
      boundBone {boneName=dyn_fleather_flaps_hip_b_01_l}
      boxExtents=<30 50 50>
      constraintSetup {angularLimitsMax=<15 0 2>, angularLimitsMin=<-15 -30 -2>, angularTarget=<1 0 0>}
      localJointOffset=<4 0 0>
     }]
   }
   n23 = Anim Dynamics - Bone: dyn_fleather_flaps_hip_b_02_l!(SimulationSpace:AnimPhysSimSpaceType Enum=Component, ComponentPose=@n22, Alpha=1, GravityOverride:Vector="0,0,-980", ExternalForce="0,0,0") -> (Pose) {
    angularDampingOverride=1
    angularSpringConstant=20000
    bAngularSpring=true
    bOverrideAngularDamping=true
    bUseGravityOverride=true
    gravityScale=0.35
    alphaBoolBlend {blendOption=Linear}
    boundBone {boneName=dyn_fleather_flaps_hip_b_02_l}
    componentAppliedLinearAccClamp=<3000 3000 3000>
    componentLinearAccScale=<0.3 0.3 0.3>
    componentLinearVelScale=<1 1 1>
    gravityOverride {z=-980}
    physicsBodyDefinitions [{
      boundBone {boneName=dyn_fleather_flaps_hip_b_02_l}
      boxExtents=<30 50 50>
      constraintSetup {angularLimitsMax=<15 0 2>, angularLimitsMin=<-15 -30 -2>, angularTarget=<1 0 0>}
      localJointOffset=<4 0 0>
     }]
   }
   n24 = Anim Dynamics - Bone: dyn_fleather_flaps_hip_a_01_l!(SimulationSpace:AnimPhysSimSpaceType Enum=Component, ComponentPose=@n23, Alpha=1, GravityOverride:Vector="0,0,-980", ExternalForce="0,0,0") -> (Pose) {
    angularDampingOverride=1
    angularSpringConstant=20000
    bAngularSpring=true
    bOverrideAngularDamping=true
    bUseGravityOverride=true
    gravityScale=0.35
    alphaBoolBlend {blendOption=Linear}
    boundBone {boneName=dyn_fleather_flaps_hip_a_01_l}
    componentAppliedLinearAccClamp=<3000 3000 3000>
    componentLinearAccScale=<0.3 0.3 0.3>
    componentLinearVelScale=<1 1 1>
    gravityOverride {z=-980}
    physicsBodyDefinitions [{
      boundBone {boneName=dyn_fleather_flaps_hip_a_01_l}
      boxExtents=<30 50 50>
      constraintSetup {angularLimitsMax=<15 0 2>, angularLimitsMin=<-15 -30 -2>, angularTarget=<1 0 0>}
      localJointOffset=<4 0 0>
     }]
   }
   n25 = Anim Dynamics - Bone: dyn_fleather_flaps_hip_a_02_l!(SimulationSpace:AnimPhysSimSpaceType Enum=Component, ComponentPose=@n24, Alpha=1, GravityOverride:Vector="0,0,-980", ExternalForce="0,0,0") -> (Pose) {
    angularDampingOverride=1
    angularSpringConstant=20000
    bAngularSpring=true
    bOverrideAngularDamping=true
    bUseGravityOverride=true
    gravityScale=0.35
    alphaBoolBlend {blendOption=Linear}
    boundBone {boneName=dyn_fleather_flaps_hip_a_02_l}
    componentAppliedLinearAccClamp=<3000 3000 3000>
    componentLinearAccScale=<0.3 0.3 0.3>
    componentLinearVelScale=<1 1 1>
    gravityOverride {z=-980}
    physicsBodyDefinitions [{
      boundBone {boneName=dyn_fleather_flaps_hip_a_02_l}
      boxExtents=<30 50 50>
      constraintSetup {angularLimitsMax=<15 0 2>, angularLimitsMin=<-15 -30 -2>, angularTarget=<1 0 0>}
      localJointOffset=<4 0 0>
     }]
   }
   n26 = Anim Dynamics - Bone: dyn_fleather_flaps_hip_a_03_l!(SimulationSpace:AnimPhysSimSpaceType Enum=Component, ComponentPose=@n25, Alpha=1, GravityOverride:Vector="0,0,-980", ExternalForce="0,0,0") -> (Pose) {
    angularDampingOverride=1
    angularSpringConstant=20000
    bAngularSpring=true
    bOverrideAngularDamping=true
    bUseGravityOverride=true
    gravityScale=0.35
    alphaBoolBlend {blendOption=Linear}
    boundBone {boneName=dyn_fleather_flaps_hip_a_03_l}
    componentAppliedLinearAccClamp=<3000 3000 3000>
    componentLinearAccScale=<0.3 0.3 0.3>
    componentLinearVelScale=<1 1 1>
    gravityOverride {z=-980}
    physicsBodyDefinitions [{
      boundBone {boneName=dyn_fleather_flaps_hip_a_03_l}
      boxExtents=<30 50 50>
      constraintSetup {angularLimitsMax=<15 0 2>, angularLimitsMin=<-15 -30 -2>, angularTarget=<1 0 0>}
      localJointOffset=<4 0 0>
     }]
   }
   n27 = Anim Dynamics - Bone: dyn_fleather_flaps_hip_a_01_m!(SimulationSpace:AnimPhysSimSpaceType Enum=Component, ComponentPose=@n26, Alpha=1, GravityOverride:Vector="0,0,-980", ExternalForce="0,0,0") -> (Pose) {
    angularDampingOverride=1
    angularSpringConstant=20000
    bAngularSpring=true
    bOverrideAngularDamping=true
    bUseGravityOverride=true
    gravityScale=0.35
    alphaBoolBlend {blendOption=Linear}
    boundBone {boneName=dyn_fleather_flaps_hip_a_01_m}
    componentAppliedLinearAccClamp=<3000 3000 3000>
    componentLinearAccScale=<0.3 0.3 0.3>
    componentLinearVelScale=<1 1 1>
    gravityOverride {z=-980}
    physicsBodyDefinitions [{
      boundBone {boneName=dyn_fleather_flaps_hip_a_01_m}
      boxExtents=<30 50 50>
      constraintSetup {angularLimitsMax=<15 0 2>, angularLimitsMin=<-15 -30 -2>, angularTarget=<1 0 0>}
      localJointOffset=<4 0 0>
     }]
   }
   n28 = Anim Dynamics - Bone: dyn_fleather_flaps_hip_a_02_m!(SimulationSpace:AnimPhysSimSpaceType Enum=Component, ComponentPose=@n27, Alpha=1, GravityOverride:Vector="0,0,-980", ExternalForce="0,0,0") -> (Pose) {
    angularDampingOverride=1
    angularSpringConstant=20000
    bAngularSpring=true
    bOverrideAngularDamping=true
    bUseGravityOverride=true
    gravityScale=0.35
    alphaBoolBlend {blendOption=Linear}
    boundBone {boneName=dyn_fleather_flaps_hip_a_02_m}
    componentAppliedLinearAccClamp=<3000 3000 3000>
    componentLinearAccScale=<0.3 0.3 0.3>
    componentLinearVelScale=<1 1 1>
    gravityOverride {z=-980}
    physicsBodyDefinitions [{
      boundBone {boneName=dyn_fleather_flaps_hip_a_02_m}
      boxExtents=<30 50 50>
      constraintSetup {angularLimitsMax=<15 0 2>, angularLimitsMin=<-15 -30 -2>, angularTarget=<1 0 0>}
      localJointOffset=<4 0 0>
     }]
   }
   n29 = Anim Dynamics - Bone: dyn_fleather_flaps_hip_a_01_r!(SimulationSpace:AnimPhysSimSpaceType Enum=Component, ComponentPose=@n28, Alpha=1, GravityOverride:Vector="0,0,-980", ExternalForce="0,0,0") -> (Pose) {
    angularDampingOverride=1
    angularSpringConstant=20000
    bAngularSpring=true
    bOverrideAngularDamping=true
    bUseGravityOverride=true
    gravityScale=0.35
    alphaBoolBlend {blendOption=Linear}
    boundBone {boneName=dyn_fleather_flaps_hip_a_01_r}
    componentAppliedLinearAccClamp=<3000 3000 3000>
    componentLinearAccScale=<0.3 0.3 0.3>
    componentLinearVelScale=<1 1 1>
    gravityOverride {z=-980}
    physicsBodyDefinitions [{
      boundBone {boneName=dyn_fleather_flaps_hip_a_01_r}
      boxExtents=<30 50 50>
      constraintSetup {angularLimitsMax=<15 0 2>, angularLimitsMin=<-15 -30 -2>, angularTarget=<1 0 0>}
      localJointOffset=<4 0 0>
     }]
   }
   n30 = Anim Dynamics - Bone: dyn_fleather_flaps_hip_a_02_r!(SimulationSpace:AnimPhysSimSpaceType Enum=Component, ComponentPose=@n29, Alpha=1, GravityOverride:Vector="0,0,-980", ExternalForce="0,0,0") -> (Pose) {
    angularDampingOverride=1
    angularSpringConstant=20000
    bAngularSpring=true
    bOverrideAngularDamping=true
    bUseGravityOverride=true
    gravityScale=0.35
    alphaBoolBlend {blendOption=Linear}
    boundBone {boneName=dyn_fleather_flaps_hip_a_02_r}
    componentAppliedLinearAccClamp=<3000 3000 3000>
    componentLinearAccScale=<0.3 0.3 0.3>
    componentLinearVelScale=<1 1 1>
    gravityOverride {z=-980}
    physicsBodyDefinitions [{
      boundBone {boneName=dyn_fleather_flaps_hip_a_02_r}
      boxExtents=<30 50 50>
      constraintSetup {angularLimitsMax=<15 0 2>, angularLimitsMin=<-15 -30 -2>, angularTarget=<1 0 0>}
      localJointOffset=<4 0 0>
     }]
   }
   n31 = Anim Dynamics - Bone: dyn_fleather_flaps_hip_b_01_r!(SimulationSpace:AnimPhysSimSpaceType Enum=Component, ComponentPose=@n30, Alpha=1, GravityOverride:Vector="0,0,-980", ExternalForce="0,0,0") -> (Pose) {
    angularDampingOverride=1
    angularSpringConstant=20000
    bAngularSpring=true
    bOverrideAngularDamping=true
    bUseGravityOverride=true
    gravityScale=0.35
    alphaBoolBlend {blendOption=Linear}
    boundBone {boneName=dyn_fleather_flaps_hip_b_01_r}
    componentAppliedLinearAccClamp=<3000 3000 3000>
    componentLinearAccScale=<0.3 0.3 0.3>
    componentLinearVelScale=<1 1 1>
    gravityOverride {z=-980}
    physicsBodyDefinitions [{
      boundBone {boneName=dyn_fleather_flaps_hip_b_01_r}
      boxExtents=<30 50 50>
      constraintSetup {angularLimitsMax=<15 0 2>, angularLimitsMin=<-15 -30 -2>, angularTarget=<1 0 0>}
      localJointOffset=<4 0 0>
     }]
   }
   n32 = Anim Dynamics - Bone: dyn_fleather_flaps_hip_b_02_r!(SimulationSpace:AnimPhysSimSpaceType Enum=Component, ComponentPose=@n31, Alpha=1, GravityOverride:Vector="0,0,-980", ExternalForce="0,0,0") -> (Pose) {
    angularDampingOverride=1
    angularSpringConstant=20000
    bAngularSpring=true
    bOverrideAngularDamping=true
    bUseGravityOverride=true
    gravityScale=0.35
    alphaBoolBlend {blendOption=Linear}
    boundBone {boneName=dyn_fleather_flaps_hip_b_02_r}
    componentAppliedLinearAccClamp=<3000 3000 3000>
    componentLinearAccScale=<0.3 0.3 0.3>
    componentLinearVelScale=<1 1 1>
    gravityOverride {z=-980}
    physicsBodyDefinitions [{
      boundBone {boneName=dyn_fleather_flaps_hip_b_02_r}
      boxExtents=<30 50 50>
      constraintSetup {angularLimitsMax=<15 0 2>, angularLimitsMin=<-15 -30 -2>, angularTarget=<1 0 0>}
      localJointOffset=<4 0 0>
     }]
   }
   n33 = Anim Dynamics - Bone: dyn_leather_flaps_Upperarm_01_l!(SimulationSpace:AnimPhysSimSpaceType Enum=Component, ComponentPose=@n32, Alpha=1, ExternalForce="0,0,0") -> (Pose) {
    angularDampingOverride=1
    angularSpringConstant=20000
    bAngularSpring=true
    bOverrideAngularDamping=true
    bUseGravityOverride=true
    gravityScale=0.35
    alphaBoolBlend {blendOption=Linear}
    boundBone {boneName=dyn_leather_flaps_Upperarm_01_l}
    componentAppliedLinearAccClamp=<3000 3000 3000>
    componentLinearAccScale=<0.3 0.3 0.3>
    componentLinearVelScale=<1 1 1>
    gravityOverride {z=-980}
    physicsBodyDefinitions [{
      boundBone {boneName=dyn_leather_flaps_Upperarm_01_l}
      boxExtents=<20 30 30>
      constraintSetup {angularLimitsMax=<15 15 15>, angularLimitsMin=<-15 -15 -15>, angularTarget=<1 0 0>}
      localJointOffset=<4 0 0>
     }]
   }
   n34 = Anim Dynamics - Bone: dyn_leather_flaps_Upperarm_02_l!(SimulationSpace:AnimPhysSimSpaceType Enum=Component, ComponentPose=@n33, Alpha=1, ExternalForce="0,0,0") -> (Pose) {
    angularDampingOverride=1
    angularSpringConstant=20000
    bAngularSpring=true
    bOverrideAngularDamping=true
    bUseGravityOverride=true
    gravityScale=0.35
    alphaBoolBlend {blendOption=Linear}
    boundBone {boneName=dyn_leather_flaps_Upperarm_02_l}
    componentAppliedLinearAccClamp=<3000 3000 3000>
    componentLinearAccScale=<0.3 0.3 0.3>
    componentLinearVelScale=<1 1 1>
    gravityOverride {z=-980}
    physicsBodyDefinitions [{
      boundBone {boneName=dyn_leather_flaps_Upperarm_02_l}
      boxExtents=<20 30 30>
      constraintSetup {angularLimitsMax=<15 15 15>, angularLimitsMin=<-15 -15 -15>, angularTarget=<1 0 0>}
      localJointOffset=<4 0 0>
     }]
   }
   n35 = Anim Dynamics - Bone: dyn_leather_flaps_Lowerarm_01_l!(SimulationSpace:AnimPhysSimSpaceType Enum=Component, ComponentPose=@n34, Alpha=1, ExternalForce="0,0,0") -> (Pose) {
    angularDampingOverride=1
    angularSpringConstant=20000
    bAngularSpring=true
    bOverrideAngularDamping=true
    bUseGravityOverride=true
    gravityScale=0.35
    alphaBoolBlend {blendOption=Linear}
    boundBone {boneName=dyn_leather_flaps_Lowerarm_01_l}
    componentAppliedLinearAccClamp=<3000 3000 3000>
    componentLinearAccScale=<0.3 0.3 0.3>
    componentLinearVelScale=<1 1 1>
    gravityOverride {z=-980}
    physicsBodyDefinitions [{
      boundBone {boneName=dyn_leather_flaps_Lowerarm_01_l}
      boxExtents=<20 30 30>
      constraintSetup {angularLimitsMax=<15 15 15>, angularLimitsMin=<-15 -15 -15>, angularTarget=<1 0 0>}
      localJointOffset=<4 0 0>
     }]
   }
   n36 = Anim Dynamics - Bone: dyn_leather_flaps_Lowerarm_02_l!(SimulationSpace:AnimPhysSimSpaceType Enum=Component, ComponentPose=@n35, Alpha=1, ExternalForce="0,0,0") -> (Pose) {
    angularDampingOverride=1
    angularSpringConstant=20000
    bAngularSpring=true
    bOverrideAngularDamping=true
    bUseGravityOverride=true
    gravityScale=0.35
    alphaBoolBlend {blendOption=Linear}
    boundBone {boneName=dyn_leather_flaps_Lowerarm_02_l}
    componentAppliedLinearAccClamp=<3000 3000 3000>
    componentLinearAccScale=<0.3 0.3 0.3>
    componentLinearVelScale=<1 1 1>
    gravityOverride {z=-980}
    physicsBodyDefinitions [{
      boundBone {boneName=dyn_leather_flaps_Lowerarm_02_l}
      boxExtents=<20 30 30>
      constraintSetup {angularLimitsMax=<15 15 15>, angularLimitsMin=<-15 -15 -15>, angularTarget=<1 0 0>}
      localJointOffset=<4 0 0>
     }]
   }
   n37 = Anim Dynamics - Bone: dyn_leather_flaps_thigh_01_l!(SimulationSpace:AnimPhysSimSpaceType Enum=Component, ComponentPose=@n36, Alpha=1, ExternalForce="0,0,0") -> (Pose) {
    angularDampingOverride=1
    angularSpringConstant=20000
    bAngularSpring=true
    bOverrideAngularDamping=true
    bUseGravityOverride=true
    gravityScale=0.35
    alphaBoolBlend {blendOption=Linear}
    boundBone {boneName=dyn_leather_flaps_thigh_01_l}
    componentAppliedLinearAccClamp=<3000 3000 3000>
    componentLinearAccScale=<0.3 0.3 0.3>
    componentLinearVelScale=<1 1 1>
    gravityOverride {z=-980}
    physicsBodyDefinitions [{
      boundBone {boneName=dyn_leather_flaps_thigh_01_l}
      boxExtents=<20 30 30>
      constraintSetup {angularLimitsMax=<15 15 15>, angularLimitsMin=<-15 -15 -15>, angularTarget=<1 0 0>}
      localJointOffset=<4 0 0>
     }]
   }
   n38 = Anim Dynamics - Bone: dyn_leather_flaps_thigh_02_l!(SimulationSpace:AnimPhysSimSpaceType Enum=Component, ComponentPose=@n37, Alpha=1, ExternalForce="0,0,0") -> (Pose) {
    angularDampingOverride=1
    angularSpringConstant=20000
    bAngularSpring=true
    bOverrideAngularDamping=true
    bUseGravityOverride=true
    gravityScale=0.35
    alphaBoolBlend {blendOption=Linear}
    boundBone {boneName=dyn_leather_flaps_thigh_02_l}
    componentAppliedLinearAccClamp=<3000 3000 3000>
    componentLinearAccScale=<0.3 0.3 0.3>
    componentLinearVelScale=<1 1 1>
    gravityOverride {z=-980}
    physicsBodyDefinitions [{
      boundBone {boneName=dyn_leather_flaps_thigh_02_l}
      boxExtents=<20 30 30>
      constraintSetup {angularLimitsMax=<15 15 15>, angularLimitsMin=<-15 -15 -15>, angularTarget=<1 0 0>}
      localJointOffset=<4 0 0>
     }]
   }
   n39 = Anim Dynamics - Bone: dyn_leather_flaps_calf_01_l!(SimulationSpace:AnimPhysSimSpaceType Enum=Component, ComponentPose=@n38, Alpha=1, ExternalForce="0,0,0") -> (Pose) {
    angularDampingOverride=1
    angularSpringConstant=20000
    bAngularSpring=true
    bOverrideAngularDamping=true
    bUseGravityOverride=true
    gravityScale=0.35
    alphaBoolBlend {blendOption=Linear}
    boundBone {boneName=dyn_leather_flaps_calf_01_l}
    componentAppliedLinearAccClamp=<3000 3000 3000>
    componentLinearAccScale=<0.3 0.3 0.3>
    componentLinearVelScale=<1 1 1>
    gravityOverride {z=-980}
    physicsBodyDefinitions [{
      boundBone {boneName=dyn_leather_flaps_calf_01_l}
      boxExtents=<20 30 30>
      constraintSetup {angularLimitsMax=<15 15 15>, angularLimitsMin=<-15 -15 -15>, angularTarget=<1 0 0>}
      localJointOffset=<4 0 0>
     }]
   }
   n40 = Anim Dynamics - Bone: dyn_leather_flaps_Lowerarm_01_r!(SimulationSpace:AnimPhysSimSpaceType Enum=Component, ComponentPose=@n39, Alpha=1, ExternalForce="0,0,0") -> (Pose) {
    angularDampingOverride=1
    angularSpringConstant=20000
    bAngularSpring=true
    bOverrideAngularDamping=true
    bUseGravityOverride=true
    gravityScale=0.35
    alphaBoolBlend {blendOption=Linear}
    boundBone {boneName=dyn_leather_flaps_Lowerarm_01_r}
    componentAppliedLinearAccClamp=<3000 3000 3000>
    componentLinearAccScale=<0.3 0.3 0.3>
    componentLinearVelScale=<1 1 1>
    gravityOverride {z=-980}
    physicsBodyDefinitions [{
      boundBone {boneName=dyn_leather_flaps_Lowerarm_01_r}
      boxExtents=<20 30 30>
      constraintSetup {angularLimitsMax=<15 15 15>, angularLimitsMin=<-15 -15 -15>, angularTarget=<1 0 0>}
      localJointOffset=<4 0 0>
     }]
   }
   n41 = Anim Dynamics - Bone: dyn_leather_flaps_thig_r!(SimulationSpace:AnimPhysSimSpaceType Enum=Component, ComponentPose=@n40, Alpha=1, ExternalForce="0,0,0") -> (Pose) {
    angularDampingOverride=1
    angularSpringConstant=20000
    bAngularSpring=true
    bOverrideAngularDamping=true
    bUseGravityOverride=true
    gravityScale=0.35
    alphaBoolBlend {blendOption=Linear}
    boundBone {boneName=dyn_leather_flaps_thig_r}
    componentAppliedLinearAccClamp=<3000 3000 3000>
    componentLinearAccScale=<0.3 0.3 0.3>
    componentLinearVelScale=<1 1 1>
    gravityOverride {z=-980}
    physicsBodyDefinitions [{
      boundBone {boneName=dyn_leather_flaps_thig_r}
      boxExtents=<40 30 30>
      constraintSetup {angularLimitsMax=<0 5 2>, angularLimitsMin=<0 -5 -35>, angularTarget=<1 0 0>}
      localJointOffset=<4 0 0>
     }]
   }
   n42 = Anim Dynamics - Bone: base_leather_flaps_foot_01_r!(SimulationSpace:AnimPhysSimSpaceType Enum=Component, ComponentPose=@n41, Alpha=1, ExternalForce="0,0,0") -> (Pose) {
    angularDampingOverride=1
    angularSpringConstant=20000
    bAngularSpring=true
    bOverrideAngularDamping=true
    bUseGravityOverride=true
    gravityScale=0.35
    alphaBoolBlend {blendOption=Linear}
    boundBone {boneName=base_leather_flaps_foot_01_r}
    componentAppliedLinearAccClamp=<3000 3000 3000>
    componentLinearAccScale=<0.3 0.3 0.3>
    componentLinearVelScale=<1 1 1>
    gravityOverride {z=-980}
    physicsBodyDefinitions [{
      boundBone {boneName=base_leather_flaps_foot_01_r}
      boxExtents=<20 30 30>
      constraintSetup {angularLimitsMax=<15 15 15>, angularLimitsMin=<-15 -15 -15>, angularTarget=<1 0 0>}
      localJointOffset=<4 0 0>
     }]
   }
   n43 = Anim Dynamics - Bone: dyn_leather_flaps_foot_01_r!(SimulationSpace:AnimPhysSimSpaceType Enum=Component, ComponentPose=@n42, Alpha=1, ExternalForce="0,0,0") -> (Pose) {
    angularDampingOverride=1
    angularSpringConstant=20000
    bAngularSpring=true
    bOverrideAngularDamping=true
    bUseGravityOverride=true
    gravityScale=0.35
    alphaBoolBlend {blendOption=Linear}
    boundBone {boneName=dyn_leather_flaps_foot_01_r}
    componentAppliedLinearAccClamp=<3000 3000 3000>
    componentLinearAccScale=<0.3 0.3 0.3>
    componentLinearVelScale=<1 1 1>
    gravityOverride {z=-980}
    physicsBodyDefinitions [{
      boundBone {boneName=dyn_leather_flaps_foot_01_r}
      boxExtents=<20 30 30>
      constraintSetup {angularLimitsMax=<15 15 15>, angularLimitsMin=<-15 -15 -15>, angularTarget=<1 0 0>}
      localJointOffset=<4 0 0>
     }]
   }
   n44 = Anim Dynamics - Bone: dyn_leather_flaps_foot_02_r!(SimulationSpace:AnimPhysSimSpaceType Enum=Component, ComponentPose=@n43, Alpha=1, ExternalForce="0,0,0") -> (Pose) {
    angularDampingOverride=1
    angularSpringConstant=20000
    bAngularSpring=true
    bOverrideAngularDamping=true
    bUseGravityOverride=true
    gravityScale=0.35
    alphaBoolBlend {blendOption=Linear}
    boundBone {boneName=dyn_leather_flaps_foot_02_r}
    componentAppliedLinearAccClamp=<3000 3000 3000>
    componentLinearAccScale=<0.3 0.3 0.3>
    componentLinearVelScale=<1 1 1>
    gravityOverride {z=-980}
    physicsBodyDefinitions [{
      boundBone {boneName=dyn_leather_flaps_foot_02_r}
      boxExtents=<20 30 30>
      constraintSetup {angularLimitsMax=<15 15 15>, angularLimitsMin=<-15 -15 -15>, angularTarget=<1 0 0>}
      localJointOffset=<4 0 0>
     }]
   }
   n45 = Anim Dynamics - Bone: dyn_leather_flaps_foot_03_r!(SimulationSpace:AnimPhysSimSpaceType Enum=Component, ComponentPose=@n44, Alpha=1, ExternalForce="0,0,0") -> (Pose) {
    angularDampingOverride=1
    angularSpringConstant=20000
    bAngularSpring=true
    bOverrideAngularDamping=true
    bUseGravityOverride=true
    gravityScale=0.35
    alphaBoolBlend {blendOption=Linear}
    boundBone {boneName=dyn_leather_flaps_foot_03_r}
    componentAppliedLinearAccClamp=<3000 3000 3000>
    componentLinearAccScale=<0.3 0.3 0.3>
    componentLinearVelScale=<1 1 1>
    gravityOverride {z=-980}
    physicsBodyDefinitions [{
      boundBone {boneName=dyn_leather_flaps_foot_03_r}
      boxExtents=<20 30 30>
      constraintSetup {angularLimitsMax=<15 15 15>, angularLimitsMin=<-15 -15 -15>, angularTarget=<1 0 0>}
      localJointOffset=<4 0 0>
     }]
   }
   n46 = "Transform (Modify) Bone - Bone: knee_offA_r"!(ComponentPose=@n45, bAlphaBoolEnabled:bool=True, Translation="0,0,4", Rotation="0,0,0", Scale="1,1,1") -> (Pose) {
    alphaInputType=Bool
    rotationMode=BMM_Additive
    rotationSpace=BCS_BoneSpace
    translationMode=BMM_Additive
    translationSpace=BCS_BoneSpace
    alphaBoolBlend {blendInTime=0.1, blendOption=Linear, blendOutTime=0.1}
    boneToModify {boneName=knee_offA_r}
    translation {z=4}
   }
   n47 = "Transform (Modify) Bone - Bone: metal_bars_03_midtop_r"(ComponentPose=@n46, Alpha=1, Translation="-20,0,0", Rotation="0,0,0", Scale="1,1,1") {
    rotationMode=BMM_Additive
    rotationSpace=BCS_BoneSpace
    translationMode=BMM_Additive
    translationSpace=BCS_BoneSpace
    alphaBoolBlend {blendInTime=0.1, blendOption=Linear, blendOutTime=0.1}
    boneToModify {boneName=metal_bars_03_midtop_r}
    translation {x=-20}
   }
   Output Pose(Result=Component To Local(ComponentPose=@n47))
  ]}
 EventGraph {nodes [
   Event BlueprintUpdateAnimation()
   > n1 = Sequence!(execute) -> (then_0, then_1, then_2:Exec)
   > n2 = Gravity Override State Function DEPRECATED(CharacterPartAnimInstance=Self-Reference(), GameTimeInSeconds=GetGameTimeInSeconds())
   > n3 = Branch(Condition=Get bIsDBNOCarried())
   > Set gravity_override(gravity_override="0, 0,-980")
   Set gravity_override(execute=@n3.else, gravity_override=@n2)
   n4 = Branch(
    execute=@n1.then_1
    Condition=OR Boolean!(A=Get bIsCrouching(), B=Get bIsCrouchSprinting(), C:bool=Get bIsJumping(), D:bool=Get bIsFalling()) -> (ReturnValue)
   )
   > Set KneeFix(KneeFix:bool=true) -> (Output_Get:bool)
   Set KneeFix(execute=@n4.else, KneeFix=false)
   n5 = Branch(execute=@n1.then_2, Condition=Get bIsInFrontEnd())
   > Set SimlationSpace(SimlationSpace:AnimPhysSimSpaceType Enum=World) -> (Output_Get:AnimPhysSimSpaceType Enum)
   > Set AlphaFE(AlphaFE:double=0.5) -> (Output_Get:double)
   > Set AlphaFE_hip(AlphaFE_hip:double=0.3) -> (Output_Get:double)
   Set SimlationSpace(execute=@n5.else, SimlationSpace=Component)
   > Set AlphaFE(AlphaFE=1)
   > Set AlphaFE_hip(AlphaFE_hip=1)
   Event OnCharacterCustomizationLoaded()
   > n6 = Cast To FortPlayerPawn(Object=GetOwningActor())
   > n7 = Sequence!(execute) -> (then_0, then_1)
   > Is Valid(InputObject=Get CitrusSpoonAxe_SkelMesh() -> (CitrusSpoonAxe_SkelMesh:Skeletal Mesh Object Reference))
   n8 = GetItemDefinitionTypeFromLoadoutWithActor!(
    Actor:Actor Object Reference=@n6
    LoadoutSlot:Gameplay Tag Structure=(TagName="LoadoutSlot.Pickaxe")
    CosmeticType:Athena Cosmetic Item Definition Class Reference=/Script/FortniteGame.AthenaPickaxeItemDefinition
   ) -> (OutCosmeticItemDef:Athena Pickaxe Item Definition Object Reference)
   > n9 = Set Current Pickaxe(Current Pickaxe:Athena Pickaxe Item Definition Object Reference=@n8) -> (Output_Get:Athena Pickaxe Item Definition Object Reference)
   n10 = "Equal (Object)"(A=@n9, B=Get CitrusSpoonAxe_ItemDef() -> (CitrusSpoonAxe_ItemDef:Athena Pickaxe Item Definition Object Reference)) {is=title K2Node_CallFunction}
   > Set gauntlet(gauntlet:double=float * float!(A="To Float (Boolean)"(InBool=@n10), B=1, C:double) -> (ReturnValue) {is=title K2Node_PromotableOperator}) -> (Output_Get:double)
   Is Valid(exec=@n7.then_1, InputObject=Get ExcitedCyanAxe_SkelMesh() -> (ExcitedCyanAxe_SkelMesh:Skeletal Mesh Object Reference))
   > n11 = Set Pickaxe(Pickaxe:Athena Pickaxe Item Definition Object Reference=@n8) -> (Output_Get:Athena Pickaxe Item Definition Object Reference)
   n12 = Get ExcitedCyanAxe_ItemDef() -> (ExcitedCyanAxe_ItemDef:Athena Pickaxe Item Definition Object Reference)
   > Set gauntlet_MetalBar(gauntlet_MetalBar:double="To Float (Boolean)"(InBool="Equal (Object)"(A=@n11, B=@n12))) -> (Output_Get:double)
   TryGetPawnOwner()
  ]}
]
variables [
 AlphaFE {type=number, default=1}
 AlphaFE_hip {type=number, default=1}
 CitrusSpoonAxe_ItemDef {type=AthenaPickaxeItemDefinition, default=/BRCosmetics/Athena/Items/Cosmetics/Pickaxes/Pickaxe_CitrusSpoon}
 CitrusSpoonAxe_SkelMesh {type=SkeletalMesh, default=/BRCosmetics/Weapons/FORT_Melee/Pickaxe_CitrusSpoon/Meshes/CitrusSpoon_Axe}
 Current Pickaxe {type=AthenaPickaxeItemDefinition, default=null}
 ExcitedCyanAxe_ItemDef {type=AthenaPickaxeItemDefinition, default=/BRCosmetics/Athena/Items/Cosmetics/Pickaxes/Pickaxe_ExcitedCyan}
 ExcitedCyanAxe_SkelMesh {type=SkeletalMesh, default=/BRCosmetics/Weapons/FORT_Melee/Pickaxe_ExcitedCyan/Meshes/ExcitedCyan_Axe}
 gauntlet {type=number}
 gauntlet_MetalBar {type=number}
 gravity_override {type=Vector}
 IsCrouching {type=number}
 KneeFix {type=boolean}
 Pickaxe {type=AthenaPickaxeItemDefinition, default=null}
 SimlationSpace {type=AnimPhysSimSpaceType}
]
